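# Material function that samples the Hermes terrain biome mask texture, outputting its UVs and four RGBA quadrant masks.
# UVs come from landscape coordinates or world XY remapped by the BiomeTextureWorldUV parameter, optionally using the position behind translucency for projection.
# Halved UVs plus half-unit offsets read the four quadrants of one texture, packing 16 channels.
# With projection enabled, a composite-derived page mask multiplies each quadrant to cut off the clamped borders.
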MaterialFunction /BRCoreMaterials/Hermes/Terrain/BiomeMask/MF_Hermes_BiomeMasks
graph {nodes [
  n1 = VectorParameter() {
   desc=This is intended to match the default landscape UVs using world position
   group=BiomeSettings
   parameterName=BiomeTextureWorldUV
   channelNames {a=LWUV-Divide, b=LWUV-Subtract, g=LWUV-Y, r=LWUV-X}
   defaultValue=<rgba 0 0 -130048 260096>
  }
  n2 = MaterialFunctionCall() {materialFunction=&WorldPositionBehindTranslucency}
  n3 = WorldPosition()
  n4 = FunctionInput(Preview=StaticBool()) {
   bUsePreviewValueAsDefault=true
   description="Default=False"
   inputName="UseProjection(translucent)"
   inputType=FunctionInput_StaticBool
   sortPriority=1
  }
  n5 = Add(
   A=ComponentMask(None=@n1) {g=true, r=true}
   B=StaticSwitch(True=ComponentMask(None=@n2.Result) {g=true, r=true}, False=TruncateLWC!(None=@n3.XY), Value=@n4)
  )
  n6 = Saturate(None=Divide(A=Subtract(A=@n5, B=@n1.LWUV-Subtract), B=@n1.LWUV-Divide))
  Composite!(InUVs=@n6, Enabled=@n4) {subgraphName=DerrivePageMask}
  n7 = FunctionInput(Preview=StaticBool()) {
   bUsePreviewValueAsDefault=true
   description="Default=False. False uses world coords"
   inputName=UseLandscapeCoords
   inputType=FunctionInput_StaticBool
  }
  n8 = StaticSwitch(True=LandscapeLayerCoords() {mappingScale=2033}, False=@n6, Value=@n7)
  FunctionOutput(None=@n8) {description=Red:Chapparal Coast, outputName=UVs, sortPriority=-1000}
  n9 = Add(A=NamedRerouteUsage(), B=Constant2Vector() {g=0.5})
  n10 = NamedRerouteUsage()
  FunctionOutput(None=Multiply(
    A=TextureSample(UVs=@n9, Tex=@n10) {desc=BottomLeft-Quadrant-C, samplerSource=SSM_Wrap_WorldGroupSettings, samplerType=SAMPLERTYPE_LinearColor}.RGBA
    B=NamedRerouteUsage()
   )) {outputName=QuadC_RGBA, sortPriority=14}
  n11 = Add(A=NamedRerouteUsage(), B=Constant2Vector() {r=0.5})
  n12 = NamedRerouteUsage()
  FunctionOutput(None=Multiply(
    A=TextureSample(UVs=@n11, Tex=@n12) {desc=UpperRight-Quadrant-B, samplerSource=SSM_Wrap_WorldGroupSettings, samplerType=SAMPLERTYPE_LinearColor}.RGBA
    B=NamedRerouteUsage()
   )) {outputName=QuadB_RGBA, sortPriority=9}
  n13 = Add(A=NamedRerouteUsage(), B=Constant2Vector() {g=0.5, r=0.5})
  n14 = NamedRerouteUsage()
  FunctionOutput(None=Multiply(
    A=TextureSample(UVs=@n13, Tex=@n14) {desc=BottomRight-Quadrant-D, samplerSource=SSM_Wrap_WorldGroupSettings, samplerType=SAMPLERTYPE_LinearColor}.RGBA
    B=NamedRerouteUsage()
   )) {outputName=QuadD_RGBA, sortPriority=19}
  n15 = NamedRerouteUsage()
  n16 = NamedRerouteUsage()
  FunctionOutput(None=Multiply(
    A=TextureSample(UVs=@n15, Tex=@n16) {desc=UpperLeft-Quadrant-A, samplerSource=SSM_Wrap_WorldGroupSettings, samplerType=SAMPLERTYPE_LinearColor}.RGBA
    B=NamedRerouteUsage()
   )) {description=Red:Chapparal Coast, outputName=QuadA_RGBA, sortPriority=4}
  NamedRerouteDeclaration(None=Multiply(A=@n8, B=Constant() {r=0.5})) {name=BaseCoord, nodeColor=<rgba 1 0 1 1>}
  n17 = TextureObjectParameter() {
   group=BiomeSettings
   parameterName=BiomeTexture
   samplerType=SAMPLERTYPE_LinearColor
   texture=&/T_Hermes_BiomeMask
  }
  NamedRerouteDeclaration(None=@n17) {name=BiomeTexture, nodeColor=<rgba 1 0 1 1>}
  n18 = Constant() {r=0.005}
  MaterialExpressionReroute_1 = Reroute(None=@n6)
  n19 = Abs(None=Subtract(A=@MaterialExpressionReroute_1.InUVs) {constB=0.5})
  n20 = MaterialFunctionCall(Float2=OneMinus(None=Multiply(A=@n19) {constB=2})) {materialFunction=&BreakOutFloat2Components}
  MaterialExpressionReroute_3 = Reroute(None=@n4)
  n21 = StaticSwitch(
   True=Multiply(A=Step(Y=@n18, X=@n20.R) {constY=0.01}, B=Step(Y=@n18, X=@n20.G) {constX=0.001, constY=0.01})
   False=Constant() {r=1}
   Value=@MaterialExpressionReroute_3.Enabled
  )
  MaterialExpressionReroute_0 = Reroute(None=@n21)
  NamedRerouteDeclaration(None=@MaterialExpressionReroute_0.Output) {name=PageMask, nodeColor=<rgba 0.875 0 1 1>}
  PinBase() {reroutePins [
    InUVs {expression=&self:MaterialExpressionReroute_1}
    Enabled {expression=&self:MaterialExpressionReroute_3}
   ]}
  PinBase!(Output=@n21) {reroutePins [Output {expression=&self:MaterialExpressionReroute_0}]}
 ]}
